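# Sound cue for the default button press in the front-end Locker UI, routed through the Fort_UI sound class.
# A single wave player feeds the root directly with the item-and-locker pressed sound at a slightly boosted volume.
# A concurrency override allows one instance at a time, stopping the oldest so rapid presses restart the sound instead of stacking.
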SoundCue /Game/Sounds/Fort_UI_Sounds/FrontEnd/Locker/Locker_Button_Default_Pressed_Cue
root=WavePlayer_0
nodes [{id=WavePlayer_0, properties {soundWaveAssetPtr=/Game/Sounds/Fort_UI_Sounds/FrontEnd/UI_FN_ItemandLocker_Pressed_01}}]
properties {
 bOverrideConcurrency=true
 maxDistance=2097152
 soundClassObject=&Fort_UI
 volumeMultiplier=1.25
 concurrencyOverrides {maxCount=1, resolutionRule=StopOldest}
}
